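# Material function in the BeakT2Fire vehicle body cosmetic folder that shifts a color's lightness toward black or white.
# InputShift is clamped to 0–1, with the midpoint returning InputColor unchanged.
# An If node selects the blend: lower values lerp from black up to the color, higher values lerp from the color to white.
# The input color and final result are not clamped, and both inputs use their preview values as defaults.
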MaterialFunction /VehicleCosmetics/Bodies/BeakT2Fire/Materials/MR_LightnessShiftNoClamp
graph {nodes [
  n1 = Constant3Vector() {constant=<rgba 1 1 1 0>}
  n2 = FunctionInput() {bUsePreviewValueAsDefault=true, inputName=InputShift, inputType=FunctionInput_Scalar, sortPriority=-1, previewValue=<1 0 0 1>}
  n3 = Clamp(None=@n2)
  n4 = Subtract(A=@n3) {constB=0.5}
  n5 = Clamp(None=Multiply(A=Clamp(None=@n4)) {constB=2})
  n6 = FunctionInput() {bUsePreviewValueAsDefault=true, inputName=InputColor, previewValue=<1 1 1 1>}
  n7 = Constant3Vector()
  n8 = Clamp(None=Multiply(A=@n3) {constB=2})
  FunctionOutput(None=If(A=@n3, "A > B"=LinearInterpolate(A=@n6, B=@n1, Alpha=@n5), "A == B"=@n6, "A < B"=LinearInterpolate(A=@n7, B=@n6, Alpha=@n8)) {constB=0.5})
 ]}
