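# Translucent unlit material for the engine boost flame on BeakT2 vehicle body skeletal meshes, faded by vertex alpha and a code-driven BoostGlowIntensity.
# Panning water normal maps distort a V-gradient flame shape and two multiplied particle-sheet samples that add sparks.
# A sharpened UV1.x mask separates inner and outer layers for color, speed and sparks.
# The fake fresnel sums camera vector world X and Y, ignoring surface normals.
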Material /VehicleCosmetics/Bodies/BeakT2Fire/Materials/M_BeakT2_EngineFlame
graph {
 properties {
  bAutomaticallySetUsageInEditor=false
  bUsedWithSkeletalMesh=true
  blendMode=BLEND_Translucent
  shadingModel=MSM_Unlit
  usedShadingModels=Unlit
 }
 nodes [
  n1 = VectorParameter() {group=EngineGlow, parameterName=BoostGlowColor, defaultValue=<rgba 1 0.2011 0.2011 1>}
  n2 = MaterialFunctionCall!(InputShift=ScalarParameter() {defaultValue=0.7, group=EngineGlow, parameterName=LightnessShift}, InputColor=@n1.RGB) {materialFunction=&/MR_LightnessShiftNoClamp}
  n3 = TextureCoordinate() {coordinateIndex=1}
  n4 = Clamp(None=ConstantBiasScale(None=ComponentMask(None=@n3) {r=true}) {bias=-0.5, scale=100})
  n5 = VertexColor()
  Multiply(
   A=Multiply(
    A=LinearInterpolate(A=@n1.RGB, B=@n2.Result, Alpha=@n4)
    B=ScalarParameter() {defaultValue=1, group=EngineGlow, parameterName=BoostGlowBrightness}
   )
   B=@n5.A
  )
  n6 = TextureCoordinate()
  n7 = TextureCoordinate() {vTiling=0.25}
  n8 = Multiply(
   A=Time()
   B=LinearInterpolate(A=ScalarParameter() {defaultValue=1, parameterName=Outer_Speed}, B=ScalarParameter() {defaultValue=1, parameterName=Inner_Speed}, Alpha=@n4)
  )
  n9 = TextureSample(UVs=Panner(Coordinate=ComponentMask(None=@n7) {g=true, r=true}, Time=@n8) {speedX=-0.05, speedY=1}) {samplerSource=SSM_Wrap_WorldGroupSettings, samplerType=SAMPLERTYPE_Normal, texture=/VehicleCosmetics/Wheels/_SharedTextures/FX/Water_02_N}
  n10 = TextureCoordinate() {vTiling=0.5}
  n11 = TextureSample(UVs=Panner(Coordinate=ComponentMask(None=@n10) {g=true, r=true}, Time=@n8) {speedX=0.02, speedY=1}) {samplerSource=SSM_Wrap_WorldGroupSettings, samplerType=SAMPLERTYPE_Normal, texture=&Water_02_N}
  n12 = ComponentMask(None=Add(A=@n9.RGB, B=@n11.RGB)) {g=true, r=true}
  n13 = Multiply(B=@n12) {constA=0.5}
  n14 = Power(Base=ComponentMask(None=Add(A=@n6, B=@n13)) {g=true}, Exp=ScalarParameter() {defaultValue=8, parameterName=GradientAmount})
  n15 = ScalarParameter() {defaultValue=2, parameterName=GradientSharpness}
  n16 = CameraVectorWS()
  n17 = Add(A=DotProduct(A=@n16, B=Constant3Vector() {constant=<rgba 0 1 0 1>}), B=DotProduct(A=@n16, B=Constant3Vector() {constant=<rgba 1 0 0 1>}))
  n18 = Power(Base=ComponentMask(None=@n6) {g=true}) {constExponent=4}
  n19 = LinearInterpolate(
   A=ScalarParameter() {defaultValue=6, parameterName=FresnelEnd}
   B=ScalarParameter() {defaultValue=3, parameterName=FresnelBase}
   Alpha=Clamp(None=Multiply(A=@n18) {constB=4})
  )
  n20 = Multiply(A=Clamp(None=Multiply(A=@n14, B=@n15)), B=Clamp(None=Power(Base=@n17, Exp=@n19)))
  n21 = TextureCoordinate() {vTiling=0.2}
  n22 = AppendVector(A=ScalarParameter() {defaultValue=1, parameterName=TileX}, B=ScalarParameter() {defaultValue=1, parameterName=TileY})
  n23 = Panner(Coordinate=ComponentMask(None=Multiply(A=@n21, B=@n22)) {g=true, r=true}, Time=@n8) {speedY=1}
  n24 = Multiply(A=Constant() {r=0.1}, B=@n12)
  n25 = TextureCoordinate() {vTiling=0.2}
  n26 = Panner(Coordinate=ComponentMask(None=Multiply(A=@n25, B=@n22)) {g=true, r=true}, Time=@n8) {speedY=2}
  n27 = Multiply(
   A=TextureSample(UVs=Add(A=@n23, B=@n24)) {samplerSource=SSM_Wrap_WorldGroupSettings, texture=/VehicleCosmetics/Wheels/_SharedTextures/FX/ParticleSheet_T}.RGB
   B=TextureSample(UVs=Add(A=@n26, B=@n24)) {samplerSource=SSM_Wrap_WorldGroupSettings, texture=&ParticleSheet_T}.RGB
  )
  n28 = LinearInterpolate(
   A=ScalarParameter() {defaultValue=1, parameterName=Outer_Sparks}
   B=ScalarParameter() {defaultValue=1, parameterName=Inner_Sparks}
   Alpha=@n4
  )
  n29 = Add(A=@n13, B=TextureCoordinate())
  n30 = Clamp(None=OneMinus(None=ComponentMask(None=@n29) {g=true}))
  n31 = Clamp(None=Multiply(A=Power(Base=@n30) {constExponent=4}) {constB=64})
  n32 = Clamp(None=@n17)
  n33 = Multiply(A=Clamp(None=Subtract(A=@n27, B=@n28)), B=Multiply(A=@n20, B=Multiply(A=@n31, B=@n32)))
  n34 = Multiply(A=ScalarParameter() {defaultValue=1, parameterName=Opacity}, B=Add(A=@n20, B=Multiply(A=@n33) {constB=64}))
  n35 = Clamp(None=ScalarParameter() {group=ForCodeUseOnly, parameterName=BoostGlowIntensity})
  Multiply(A=@n5.A, B=Clamp(None=LinearInterpolate(B=@n34, Alpha=@n35)))
 ]
}
