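# Translucent, unlit, two-sided fire material for the Beak T2 Fire vehicle body cosmetic, usable on skeletal meshes.
# Three panning samples of one flow normal map, timed by Speed, distort UVs under a V gradient to warp a radial flame shape.
# Opacity combines that shape with a V-edge fade and a tangent-space camera-angle fade that hides edge-on surfaces.
# Emissive blends along V from an alpha-boosted, squared FireColor to a softer one, masked by the shape.
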Material /VehicleCosmetics/Bodies/BeakT2Fire/Materials/M_Beak_T2_Fire
graph {
 properties {
  bAutomaticallySetUsageInEditor=false
  bUsedWithSkeletalMesh=true
  blendMode=BLEND_Translucent
  shadingModel=MSM_Unlit
  twoSided=true
  usedShadingModels=Unlit
 }
 nodes [
  n1 = TextureCoordinate() {vTiling=0.75}
  n2 = TextureCoordinate() {uTiling=0.5}
  n3 = ComponentMask(None=Multiply(A=Time(), B=ScalarParameter() {defaultValue=2, group=Lights, parameterName=Speed})) {r=true}
  n4 = TextureSample(UVs=Panner(Coordinate=@n2, Time=@n3) {speedY=0.3}) {
   samplerSource=SSM_Wrap_WorldGroupSettings
   samplerType=SAMPLERTYPE_Normal
   texture=/VehicleCosmetics/Wheels/_SharedTextures/FX/Flow_03_nomips
  }
  n5 = TextureSample(UVs=Panner(Coordinate=Multiply(A=@n2) {constB=1.5}, Time=@n3) {speedX=-0.01, speedY=0.2}) {
   samplerSource=SSM_Wrap_WorldGroupSettings
   samplerType=SAMPLERTYPE_Normal
   texture=&Flow_03_nomips
  }
  n6 = TextureSample(UVs=Panner(Coordinate=Multiply(A=@n2) {constB=0.5}, Time=@n3) {speedY=0.1}) {
   samplerSource=SSM_Wrap_WorldGroupSettings
   samplerType=SAMPLERTYPE_Normal
   texture=&Flow_03_nomips
  }
  n7 = ComponentMask(None=Add(A=Add(A=@n4.RGB, B=@n5.RGB), B=@n6.RGB)) {g=true, r=true}
  n8 = ComponentMask(None=@n1) {g=true}
  n9 = Clamp(None=ConstantBiasScale(None=OneMinus(None=@n8)) {bias=-0.2, scale=1.5})
  n10 = Add(A=Add(A=@n1, B=Multiply(A=@n7, B=@n9))) {constB=-0.5}
  n11 = OneMinus(None=DotProduct(A=@n10, B=@n10))
  n12 = Clamp(None=ConstantBiasScale(None=@n11) {bias=-0.85, scale=6})
  n13 = Clamp(None=ConstantBiasScale(None=@n8) {bias=0, scale=8})
  n14 = Multiply(A=@n13, B=@n13)
  n15 = Transform(None=CameraVectorWS()) {transformSourceType=TRANSFORMSOURCE_World, transformType=TRANSFORM_Tangent}
  n16 = DotProduct(A=@n15, B=Constant3Vector() {constant=<rgba 0 0 -1 1>})
  n17 = DotProduct(A=@n15, B=Constant3Vector() {constant=<rgba 0 0 1 1>})
  n18 = Add(A=Clamp(None=ConstantBiasScale(None=@n16) {bias=-0.2, scale=3}), B=Clamp(None=ConstantBiasScale(None=@n17) {bias=-0.2, scale=3}))
  NamedRerouteDeclaration(None=Multiply(A=Multiply(A=@n12, B=@n14), B=Clamp(None=@n18))) {name=Opacity, nodeColor=<rgba 0.8984 0 1 1>}
  n19 = VectorParameter() {group=Lights, parameterName=FireColor, defaultValue=<rgba 1 0.35 0.1 8>}
  n20 = Multiply(A=Multiply(B=@n19.A) {constA=0.75}, B=Multiply(A=@n19.RGB, B=@n19.RGB))
  n21 = Multiply(A=@n20, B=@n20)
  n22 = Multiply(A=@n19.RGB, B=@n19.RGB)
  n23 = Clamp(None=ConstantBiasScale(None=@n8) {bias=0, scale=1.5})
  NamedRerouteDeclaration(None=Multiply(A=LinearInterpolate(A=@n21, B=@n22, Alpha=@n23), B=Clamp(None=@n11))) {name=Emissive, nodeColor=<rgba 0 1 0.5781 1>}
  NamedRerouteUsage()
  NamedRerouteUsage()
 ]
}
