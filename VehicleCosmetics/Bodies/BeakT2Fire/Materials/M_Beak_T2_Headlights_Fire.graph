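# Animated headlight material for the Beak T2 Fire car body, switchable between fire and ice looks through the ICE ON/OFF parameter.
# The parameter blends fire and ice colors, shaped by a vertical UV gradient between darkened and full color.
# Three differently scaled, panned samples of one flow normal map sum into scrolling noise, desaturated for fire or inverted red channel for ice.
# That noise, scaled by brightness and squared color, is added to the tint.
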Material /VehicleCosmetics/Bodies/BeakT2Fire/Materials/M_Beak_T2_Headlights_Fire
graph {
 properties {bAutomaticallySetUsageInEditor=false, bUsedWithSkeletalMesh=true, usedShadingModels=Default Lit}
 nodes [
  n1 = ScalarParameter() {group=Lights, parameterName=ICE ON/OFF}
  n2 = LinearInterpolate(
   A=VectorParameter() {group=Lights, parameterName=HeadlightColorFire, defaultValue=<rgba 0.949 0.482 0.0781 1>}.RGB
   B=VectorParameter() {group=Lights, parameterName=HeadlightColorIce, defaultValue=<rgba 0.4928 0.6218 0.949 1>}.RGB
   Alpha=@n1
  )
  n3 = Multiply(A=@n2, B=@n2)
  n4 = Multiply(A=@n3, B=@n3)
  n5 = Add(A=TextureCoordinate() {vTiling=3}, B=Constant2Vector() {g=-1.875})
  n6 = LinearInterpolate(A=@n4, B=@n2, Alpha=Clamp(None=ComponentMask(None=@n5) {g=true}))
  n7 = TextureCoordinate() {uTiling=16, vTiling=12}
  n8 = ComponentMask(None=Multiply(A=Time(), B=ScalarParameter() {defaultValue=2, group=Lights, parameterName=Speed})) {r=true}
  n9 = TextureSample(UVs=Panner(Coordinate=@n7, Time=@n8) {speedY=-0.3}) {
   samplerSource=SSM_Wrap_WorldGroupSettings
   samplerType=SAMPLERTYPE_Normal
   texture=/VehicleCosmetics/Wheels/_SharedTextures/FX/Flow_03_nomips
  }
  n10 = TextureSample(UVs=Panner(Coordinate=Multiply(A=@n7) {constB=1.5}, Time=@n8) {speedX=-0.01, speedY=-0.2}) {
   samplerSource=SSM_Wrap_WorldGroupSettings
   samplerType=SAMPLERTYPE_Normal
   texture=&Flow_03_nomips
  }
  n11 = TextureSample(UVs=Panner(Coordinate=Multiply(A=@n7) {constB=0.5}, Time=@n8) {speedY=-0.1}) {
   samplerSource=SSM_Wrap_WorldGroupSettings
   samplerType=SAMPLERTYPE_Normal
   texture=&Flow_03_nomips
  }
  n12 = Add(A=Add(A=@n9.RGB, B=@n10.RGB), B=@n11.RGB)
  n13 = ConstantBiasScale(None=Desaturation(None=@n12) {luminanceFactors=<rgba 0.3 0.59 0.11 1>}) {bias=0.55, scale=1}
  n14 = ConstantBiasScale(None=OneMinus(None=ComponentMask(None=@n12) {r=true})) {bias=-0.65}
  n15 = Multiply(
   A=Multiply(A=ScalarParameter() {defaultValue=5, group=Lights, parameterName=HeadlightBrightness}, B=Multiply(A=@n6, B=@n6))
   B=LinearInterpolate(A=Clamp(None=@n13), B=Clamp(None=@n14), Alpha=@n1)
  )
  Add(A=@n6, B=@n15)
  Add(A=@n15, B=@n4)
 ]
}
