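# Physics asset for the BeakT2Fire vehicle cosmetic body mesh (LOD0, BR variant), giving it a simplified collision shape.
# A single body on the frame bone holds one large capsule, offset from the bone origin and rotated to lie along roughly the bone's X axis.
# There are no constraints, so the asset defines no jointed or ragdoll parts.
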PhysicsAsset /VehicleCosmetics/Bodies/BeakT2Fire/Meshes/SM_BeakT2Fire_LOD0_Physics_BR
bodies {frame [{shapeType=Capsule, length=253.233, radius=115.0131, center=<36.2125 -53.1297 0.0008>, rotation=<rot 0 93.0923 -89.9996>}]}
constraints []
